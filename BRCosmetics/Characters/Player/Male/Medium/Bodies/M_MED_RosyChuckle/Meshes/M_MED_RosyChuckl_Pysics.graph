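# Physics asset for the RosyChuckle medium male body cosmetic, built around four dynamic jacket bone chains (front and side, left and right).
# Capsules on legs, feet, pelvis and lower spine, an upper-spine box, and an oversized root capsule below the origin form the body shapes.
# Jacket segments mostly use narrow swing/twist limits, with free-rotation cross-constraints tying front to side chains.
# Chain anchors are asymmetric: front right to root, front left to spine_03, sides to pelvis.
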PhysicsAsset /BRCosmetics/Characters/Player/Male/Medium/Bodies/M_MED_RosyChuckle/Meshes/M_MED_RosyChuckl_Pysics
bodies {
 calf_l [{shapeType=Capsule, length=34.4941, radius=12.8059, shapeName=calf_l_capsule, center=<-23.5848 -0.9162 0>, rotation=<rot 0 87.2876 -90>}]
 calf_r [{shapeType=Capsule, length=34.4941, radius=12.8059, shapeName=calf_r_capsule, center=<23.5848 0.9162 0>, rotation=<rot 0 87.2876 90>}]
 dyn_jacket_01_fr_l [{shapeType=Capsule, length=20, radius=10, shapeName=dyn_jacket_01_fr_l_capsule, center=<13.7346 0 9.7403>, rotation=<rot 0.0001 0.0001 90>}]
 dyn_jacket_01_fr_r [{shapeType=Capsule, length=15, radius=7, shapeName=dyn_jacket_01_fr_r_capsule, center=<13.9724 0 7.8387>, rotation=<rot 5 -89.9998 -90.0003>}]
 dyn_jacket_01_side_l [{
   shapeType=Capsule
   length=12
   radius=11.5
   shapeName=dyn_jacket_01_side_l_capsule
   center=<16.2513 -8.2708 9.1255>
   rotation=<rot 0 0.0001 -110>
  }]
 dyn_jacket_01_side_r [{
   shapeType=Capsule
   length=17.5
   radius=9.5
   shapeName=dyn_jacket_01_side_r_capsule
   center=<13 0 12.8819>
   rotation=<rot -0.0002 -0.0011 105>
  }]
 dyn_jacket_02_fr_l [{shapeType=Capsule, length=20, radius=10, shapeName=dyn_jacket_02_fr_l_capsule, center=<13.7346 0 9.2095>, rotation=<rot 0.0002 0.0002 90>}]
 dyn_jacket_02_fr_r [{
   shapeType=Capsule
   length=21.5
   radius=8.5
   shapeName=dyn_jacket_02_fr_r_capsule
   center=<11.7816 0.9602 9.426>
   rotation=<rot 0.0004 -0.0004 -93.1159>
  }]
 dyn_jacket_02_side_l [{
   shapeType=Capsule
   length=12
   radius=11.5
   shapeName=dyn_jacket_02_side_l_capsule
   center=<16.2513 -6.9433 5.4783>
   rotation=<rot 0 0.0001 -110>
  }]
 dyn_jacket_02_side_r [{
   shapeType=Capsule
   length=8.5
   radius=11
   shapeName=dyn_jacket_02_side_r_capsule
   center=<20.7436 0 11.7884>
   rotation=<rot -0.0001 -0.0004 90.0001>
  }]
 dyn_jacket_03_fr_l [{
   shapeType=Capsule
   length=20
   radius=10
   shapeName=dyn_jacket_03_fr_l_capsule
   center=<13.7346 -0.0728 8.6178>
   rotation=<rot 0.0002 0.0002 95>
  }]
 dyn_jacket_03_fr_r [{
   shapeType=Capsule
   length=19
   radius=9
   shapeName=dyn_jacket_03_fr_r_capsule
   center=<11.034 0.6319 9.9762>
   rotation=<rot 0.0003 -0.0004 -99.4791>
  }]
 dyn_jacket_03_side_l [{
   shapeType=Capsule
   length=12
   radius=11.5
   shapeName=dyn_jacket_03_side_l_capsule
   center=<16.2513 -5.8883 2.5796>
   rotation=<rot 0 0.0001 -110>
  }]
 dyn_jacket_03_side_r [{
   shapeType=Capsule
   length=13.5
   radius=11.5
   shapeName=dyn_jacket_03_side_r_capsule
   center=<12.7403 0 12.9625>
   rotation=<rot -0.0001 -0.0004 95.0001>
  }]
 dyn_jacket_04_fr_l [{
   shapeType=Capsule
   length=20
   radius=10
   shapeName=dyn_jacket_04_fr_l_capsule
   center=<13.7346 -0.0597 8.4694>
   rotation=<rot 0.0002 0.0002 95>
  }]
 dyn_jacket_04_fr_r [{
   shapeType=Capsule
   length=21
   radius=9
   shapeName=dyn_jacket_04_fr_r_capsule
   center=<11.034 0.8423 7.703>
   rotation=<rot 0.0003 -0.0003 -87.9664>
  }]
 dyn_jacket_04_side_l [{
   shapeType=Capsule
   length=12
   radius=11.5
   shapeName=dyn_jacket_04_side_l_capsule
   center=<16.2513 -4.3704 -1.5907>
   rotation=<rot 0 0.0001 -110>
  }]
 dyn_jacket_04_side_r [{
   shapeType=Capsule
   length=13.5
   radius=11.5
   shapeName=dyn_jacket_04_side_r_capsule
   center=<12.7403 1 10.9898>
   rotation=<rot -0.0001 -0.0003 105.0001>
  }]
 dyn_jacket_05_fr_l [{shapeType=Capsule, length=20, radius=10, shapeName=dyn_jacket_05_fr_l_capsule, center=<13.7346 0 9.7403>, rotation=<rot 0.0001 0.0002 95>}]
 dyn_jacket_05_fr_r [{
   shapeType=Capsule
   length=28.5
   radius=8
   shapeName=dyn_jacket_05_fr_r_capsule
   center=<9.8344 0.6472 5.4724>
   rotation=<rot 0.4355 -4.9814 -80.0188>
  }]
 dyn_jacket_base_fr_l [{shapeType=Capsule, length=0.505, radius=0.505, shapeName=dyn_jacket_base_fr_l_capsule}]
 dyn_jacket_base_fr_r [{shapeType=Capsule, length=0.505, radius=0.505, shapeName=dyn_jacket_base_fr_r_capsule, center=<0 -12.3182 0>, rotation=<rot 0 -90 180>}]
 dyn_jacket_base_side_l [{shapeType=Capsule, length=0.505, radius=0.505, shapeName=dyn_jacket_base_side_l_capsule, rotation=<rot 0 -90 0>}]
 dyn_jacket_base_side_r [{shapeType=Capsule, length=0.505, radius=0.505, shapeName=dyn_jacket_base_side_r_capsule}]
 foot_l [{
   shapeType=Capsule
   length=16.4958
   radius=10.1374
   shapeName=foot_l_capsule
   center=<-7.3678 4.4113 -0.4545>
   rotation=<rot 0 3.275 -89.3193>
  }]
 foot_r [{shapeType=Capsule, length=15.6047, radius=11.0284, shapeName=foot_r_capsule, center=<7.3678 -4.4113 0.4545>, rotation=<rot 0 3.275 90.6807>}]
 pelvis [{shapeType=Capsule, length=11.9, radius=12.5, shapeName=pelvis_capsule, rotation=<rot 0 90 0>}]
 root [{shapeType=Capsule, length=2.609, radius=196.391, shapeName=root_capsule, center=<0 0 -186.3833>, rotation=<rot 0 -90 0>}]
 spine_01 [{shapeType=Capsule, length=2.1048, radius=13.7857, shapeName=spine_01_capsule, rotation=<rot 90 0 0>}]
 spine_03 [{shapeType=Box, extentX=13.8142, extentY=16.3306, extentZ=23.1668, shapeName=spine_03_box, center=<0.137 5.1932 0>, rotation=<rot 0 9.9999 0>}]
 thigh_l [{shapeType=Capsule, length=28.3, radius=9.2, shapeName=thigh_l_capsule, center=<-20.0472 0.6369 -0.6829>, rotation=<rot 0 94.2351 -90>}]
 thigh_r [{shapeType=Capsule, length=26.3, radius=11.2, shapeName=thigh_r_capsule, center=<20.0472 -0.6369 0.6829>, rotation=<rot 0 94.2351 90>}]
}
constraints [
 {
  bone1Name=calf_r
  bone2Name=thigh_r
  swing1LimitDegrees=45
  swing1Motion=Free
  swing2LimitDegrees=45
  swing2Motion=Free
  twistLimitDegrees=45
  twistMotion=Free
 }
 {
  bone1Name=calf_l
  bone2Name=thigh_l
  swing1LimitDegrees=45
  swing1Motion=Free
  swing2LimitDegrees=45
  swing2Motion=Free
  twistLimitDegrees=45
  twistMotion=Free
 }
 {
  bone1Name=spine_01
  bone2Name=pelvis
  swing1LimitDegrees=45
  swing1Motion=Free
  swing2LimitDegrees=45
  swing2Motion=Free
  twistLimitDegrees=45
  twistMotion=Free
 }
 {
  bone1Name=foot_r
  bone2Name=calf_r
  swing1LimitDegrees=45
  swing1Motion=Free
  swing2LimitDegrees=45
  swing2Motion=Free
  twistLimitDegrees=45
  twistMotion=Free
 }
 {
  bone1Name=foot_l
  bone2Name=calf_l
  swing1LimitDegrees=45
  swing1Motion=Free
  swing2LimitDegrees=45
  swing2Motion=Free
  twistLimitDegrees=45
  twistMotion=Free
 }
 {
  bone1Name=spine_03
  bone2Name=spine_01
  swing1LimitDegrees=45
  swing1Motion=Free
  swing2LimitDegrees=45
  swing2Motion=Free
  twistLimitDegrees=45
  twistMotion=Free
 }
 {bone1Name=dyn_jacket_01_fr_r, bone2Name=dyn_jacket_base_fr_r, swing1LimitDegrees=22, swing2LimitDegrees=22, twistLimitDegrees=5}
 {bone1Name=dyn_jacket_02_fr_r, bone2Name=dyn_jacket_01_fr_r, swing1LimitDegrees=22, swing2LimitDegrees=22, twistLimitDegrees=5}
 {bone1Name=dyn_jacket_03_fr_r, bone2Name=dyn_jacket_02_fr_r, swing1LimitDegrees=22, swing2LimitDegrees=22, twistLimitDegrees=5}
 {
  bone1Name=dyn_jacket_04_fr_r
  bone2Name=dyn_jacket_03_fr_r
  swing1LimitDegrees=15
  swing2LimitDegrees=20
  twistLimitDegrees=10
  twistMotion=Locked
 }
 {bone1Name=dyn_jacket_05_fr_r, bone2Name=dyn_jacket_04_fr_r, swing1LimitDegrees=11, swing2LimitDegrees=11, twistLimitDegrees=5}
 {
  bone1Name=dyn_jacket_base_side_l
  bone2Name=pelvis
  swing1LimitDegrees=45
  swing1Motion=Locked
  swing2LimitDegrees=45
  swing2Motion=Locked
  twistLimitDegrees=45
  twistMotion=Locked
 }
 {bone1Name=dyn_jacket_01_side_l, bone2Name=dyn_jacket_base_side_l, swing1LimitDegrees=11, swing2LimitDegrees=11, twistLimitDegrees=5}
 {bone1Name=dyn_jacket_02_side_l, bone2Name=dyn_jacket_01_side_l, swing1LimitDegrees=22, swing2LimitDegrees=22, twistLimitDegrees=5}
 {bone1Name=dyn_jacket_03_side_l, bone2Name=dyn_jacket_02_side_l, swing1LimitDegrees=22, swing2LimitDegrees=22, twistLimitDegrees=5}
 {bone1Name=dyn_jacket_04_side_l, bone2Name=dyn_jacket_03_side_l, swing1LimitDegrees=11, swing2LimitDegrees=11, twistLimitDegrees=5}
 {
  bone1Name=dyn_jacket_base_side_r
  bone2Name=pelvis
  swing1LimitDegrees=45
  swing1Motion=Free
  swing2LimitDegrees=45
  swing2Motion=Free
  twistLimitDegrees=45
  twistMotion=Free
 }
 {
  bone1Name=dyn_jacket_01_side_r
  bone2Name=dyn_jacket_base_side_r
  swing1LimitDegrees=20
  swing1Motion=Locked
  swing2LimitDegrees=20
  swing2Motion=Locked
  twistLimitDegrees=5
  twistMotion=Locked
 }
 {bone1Name=dyn_jacket_02_side_r, bone2Name=dyn_jacket_01_side_r, swing1LimitDegrees=22, swing2LimitDegrees=22, twistLimitDegrees=5}
 {bone1Name=dyn_jacket_03_side_r, bone2Name=dyn_jacket_02_side_r, swing1LimitDegrees=22, swing2LimitDegrees=22, twistLimitDegrees=5}
 {bone1Name=dyn_jacket_04_side_r, bone2Name=dyn_jacket_03_side_r, swing1LimitDegrees=11, swing2LimitDegrees=11, twistLimitDegrees=5}
 {
  bone1Name=pelvis
  bone2Name=root
  swing1LimitDegrees=45
  swing1Motion=Free
  swing2LimitDegrees=45
  swing2Motion=Free
  twistLimitDegrees=45
  twistMotion=Free
 }
 {
  bone1Name=thigh_l
  bone2Name=root
  swing1LimitDegrees=45
  swing1Motion=Free
  swing2LimitDegrees=45
  swing2Motion=Free
  twistLimitDegrees=45
  twistMotion=Free
 }
 {
  bone1Name=thigh_r
  bone2Name=root
  swing1LimitDegrees=45
  swing1Motion=Free
  swing2LimitDegrees=45
  swing2Motion=Free
  twistLimitDegrees=45
  twistMotion=Free
 }
 {
  bone1Name=dyn_jacket_base_fr_r
  bone2Name=root
  swing1LimitDegrees=45
  swing1Motion=Free
  swing2LimitDegrees=45
  swing2Motion=Free
  twistLimitDegrees=45
  twistMotion=Free
 }
 {
  bone1Name=dyn_jacket_base_fr_l
  bone2Name=spine_03
  swing1LimitDegrees=45
  swing1Motion=Free
  swing2LimitDegrees=45
  swing2Motion=Free
  twistLimitDegrees=45
  twistMotion=Free
 }
 {
  bone1Name=dyn_jacket_01_fr_l
  bone2Name=dyn_jacket_base_fr_l
  swing1LimitDegrees=22
  swing2LimitDegrees=22
  twistLimitDegrees=45
  twistMotion=Locked
 }
 {
  bone1Name=dyn_jacket_02_fr_l
  bone2Name=dyn_jacket_01_fr_l
  swing1LimitDegrees=22
  swing2LimitDegrees=22
  twistLimitDegrees=45
  twistMotion=Locked
 }
 {
  bone1Name=dyn_jacket_03_fr_l
  bone2Name=dyn_jacket_02_fr_l
  swing1LimitDegrees=22
  swing2LimitDegrees=22
  twistLimitDegrees=45
  twistMotion=Locked
 }
 {
  bone1Name=dyn_jacket_04_fr_l
  bone2Name=dyn_jacket_03_fr_l
  swing1LimitDegrees=15
  swing2LimitDegrees=20
  twistLimitDegrees=10
  twistMotion=Locked
 }
 {
  bone1Name=dyn_jacket_05_fr_l
  bone2Name=dyn_jacket_04_fr_l
  swing1LimitDegrees=11
  swing2LimitDegrees=15
  twistLimitDegrees=45
  twistMotion=Locked
 }
 {
  bone1Name=dyn_jacket_05_fr_l
  bone2Name=dyn_jacket_04_side_l
  swing1LimitDegrees=45
  swing1Motion=Free
  swing2LimitDegrees=45
  swing2Motion=Free
  twistLimitDegrees=45
  twistMotion=Free
 }
 {
  bone1Name=dyn_jacket_05_fr_r
  bone2Name=dyn_jacket_04_side_r
  swing1LimitDegrees=45
  swing1Motion=Free
  swing2LimitDegrees=45
  swing2Motion=Free
  twistLimitDegrees=45
  twistMotion=Free
 }
 {
  bone1Name=dyn_jacket_01_side_r
  bone2Name=dyn_jacket_02_fr_r
  swing1LimitDegrees=45
  swing1Motion=Free
  swing2LimitDegrees=45
  swing2Motion=Free
  twistLimitDegrees=45
  twistMotion=Free
 }
 {
  bone1Name=dyn_jacket_03_side_l
  bone2Name=dyn_jacket_03_fr_l
  swing1LimitDegrees=45
  swing1Motion=Free
  swing2LimitDegrees=45
  swing2Motion=Free
  twistLimitDegrees=45
  twistMotion=Free
 }
]
