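# Front-end physics asset for the RosyChuckle medium male body, covering the lower body and a two-sided dynamic jacket back.
# Capsules on pelvis, thighs, calves and feet form the body, with free angular motion at the hips.
# Left and right three-segment jacket chains hang from pelvis-locked base bones with tight swing limits and mostly locked twist.
# Two constraints join the chain tips across the back, one angularly free and one limited.
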PhysicsAsset /BRCosmetics/Characters/Player/Male/Medium/Bodies/M_MED_RosyChuckle/Meshes/M_MED_RosyChuckle_Physics_FrontEnd
bodies {
 calf_l [{shapeType=Capsule, length=20.6458, radius=10.3826, center=<-19.7227 -1.2259 -0.5968>, rotation=<rot 0 -91.8539 -89.8577>}]
 calf_r [{shapeType=Capsule, length=20.1555, radius=10.3242, center=<20.2878 1.3914 0.5245>, rotation=<rot 0 -90.4056 -88.8823>}]
 dyn_jacket_01_back_l [{shapeType=Capsule, length=17, radius=8.5, center=<0.0001 -3.0606 7.4602>, rotation=<rot -0.0004 0.0002 65>}]
 dyn_jacket_01_back_r [{shapeType=Capsule, length=17, radius=8.5, center=<1.7292 -4.0251 6.3988>, rotation=<rot -3.7868 4.085 105>}]
 dyn_jacket_02_back_l [{shapeType=Capsule, length=18, radius=7.5, center=<-0.7177 -4.0245 10.3513>, rotation=<rot -0.0004 0.0001 65>}]
 dyn_jacket_02_back_r [{shapeType=Capsule, length=18, radius=7.5, center=<1.7318 -4.0162 9.1123>, rotation=<rot -3.7868 4.085 105>}]
 dyn_jacket_03_back_l [{shapeType=Capsule, length=16, radius=9.5, center=<3.1093 -4.3688 11.0544>, rotation=<rot -0.0003 0.0001 65>}]
 dyn_jacket_03_back_r [{shapeType=Capsule, length=16, radius=9.5, center=<4.119 -3.584 11.5149>, rotation=<rot -3.7868 4.085 105>}]
 dyn_jacket_base_back_l [{shapeType=Capsule, length=1, radius=1, center=<0 -2.8139 0>}]
 dyn_jacket_base_back_r [{shapeType=Capsule, length=1, radius=1, center=<0 -2.8139 0>}]
 foot_l [{shapeType=Capsule, length=12.8563, radius=13.5854, center=<-2.5197 5.2749 -0.0889>, rotation=<rot 0 3.5581 -87.9004>}]
 foot_r [{shapeType=Capsule, length=12.8563, radius=13.5854, center=<2.5197 -5.2749 0.0889>, rotation=<rot 0 3.5581 92.0996>}]
 pelvis [{shapeType=Capsule, length=11.1629, radius=12.3569, shapeName=pelvis_capsule, center=<0 2.6649 0>}]
 thigh_l [{shapeType=Capsule, length=20.4237, radius=10.7592, center=<-19.8798 -0.6163 0.6151>, rotation=<rot 0 -88.4152 -84.6098>}]
 thigh_r [{shapeType=Capsule, length=20.4237, radius=10.7592, center=<19.8798 0.6163 -0.6151>, rotation=<rot 0 -88.4152 95.3902>}]
}
constraints [
 {bone1Name=calf_l, bone2Name=thigh_l, swing1LimitDegrees=45, swing2LimitDegrees=45, twistLimitDegrees=45}
 {bone1Name=foot_l, bone2Name=calf_l, swing1LimitDegrees=45, swing2LimitDegrees=45, twistLimitDegrees=45}
 {bone1Name=calf_r, bone2Name=thigh_r, swing1LimitDegrees=45, swing2LimitDegrees=45, twistLimitDegrees=45}
 {bone1Name=foot_r, bone2Name=calf_r, swing1LimitDegrees=45, swing2LimitDegrees=45, twistLimitDegrees=45}
 {
  bone1Name=dyn_jacket_01_back_l
  bone2Name=dyn_jacket_base_back_l
  swing1LimitDegrees=5
  swing2LimitDegrees=20
  twistLimitDegrees=45
  twistMotion=Locked
 }
 {
  bone1Name=dyn_jacket_02_back_l
  bone2Name=dyn_jacket_01_back_l
  swing1LimitDegrees=20
  swing2LimitDegrees=35
  twistLimitDegrees=10
  twistMotion=Locked
 }
 {bone1Name=dyn_jacket_03_back_l, bone2Name=dyn_jacket_02_back_l, swing1LimitDegrees=5, swing2LimitDegrees=30, twistLimitDegrees=5}
 {
  bone1Name=thigh_l
  bone2Name=pelvis
  swing1LimitDegrees=45
  swing1Motion=Free
  swing2LimitDegrees=45
  swing2Motion=Free
  twistLimitDegrees=45
  twistMotion=Free
 }
 {
  bone1Name=thigh_r
  bone2Name=pelvis
  swing1LimitDegrees=45
  swing1Motion=Free
  swing2LimitDegrees=45
  swing2Motion=Free
  twistLimitDegrees=45
  twistMotion=Free
 }
 {
  bone1Name=dyn_jacket_base_back_l
  bone2Name=pelvis
  swing1LimitDegrees=45
  swing1Motion=Locked
  swing2LimitDegrees=45
  swing2Motion=Locked
  twistLimitDegrees=45
  twistMotion=Locked
 }
 {
  bone1Name=dyn_jacket_01_back_r
  bone2Name=dyn_jacket_base_back_r
  swing1LimitDegrees=5
  swing2LimitDegrees=20
  twistLimitDegrees=45
  twistMotion=Locked
 }
 {
  bone1Name=dyn_jacket_02_back_r
  bone2Name=dyn_jacket_01_back_r
  swing1LimitDegrees=20
  swing2LimitDegrees=35
  twistLimitDegrees=10
  twistMotion=Locked
 }
 {bone1Name=dyn_jacket_03_back_r, bone2Name=dyn_jacket_02_back_r, swing1LimitDegrees=5, swing2LimitDegrees=30, twistLimitDegrees=5}
 {
  bone1Name=dyn_jacket_base_back_r
  bone2Name=pelvis
  swing1LimitDegrees=45
  swing1Motion=Locked
  swing2LimitDegrees=45
  swing2Motion=Locked
  twistLimitDegrees=45
  twistMotion=Locked
 }
 {
  bone1Name=dyn_jacket_03_back_r
  bone2Name=dyn_jacket_03_back_l
  swing1LimitDegrees=45
  swing1Motion=Free
  swing2LimitDegrees=45
  swing2Motion=Free
  twistLimitDegrees=45
  twistMotion=Free
 }
 {bone1Name=dyn_jacket_03_back_r, bone2Name=dyn_jacket_03_back_l, swing1LimitDegrees=45, swing2LimitDegrees=45, twistLimitDegrees=45}
]
