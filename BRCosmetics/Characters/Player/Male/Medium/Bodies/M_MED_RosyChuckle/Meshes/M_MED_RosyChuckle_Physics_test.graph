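# Test-named physics asset for the RosyChuckle medium male body, built around the two dynamic back jacket chains.
# Capsules cover pelvis, legs and feet, plus a very large root capsule offset below the origin.
# Each three-link jacket chain hangs from a base body locked to the pelvis, with tight swing limits and mostly locked twist.
# Two constraints, one of them fully free, link the left and right chain tips.
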PhysicsAsset /BRCosmetics/Characters/Player/Male/Medium/Bodies/M_MED_RosyChuckle/Meshes/M_MED_RosyChuckle_Physics_test
bodies {
 calf_l [{shapeType=Capsule, length=19.7869, radius=11.2415, center=<-19.7227 -1.2259 -0.5968>, rotation=<rot 0 -91.8539 -89.8577>}]
 calf_r [{shapeType=Capsule, length=19.2966, radius=11.1831, center=<20.2878 1.3914 0.5245>, rotation=<rot 0 -90.4056 -88.8823>}]
 dyn_jacket_01_back_l [{shapeType=Capsule, length=15, radius=10.5, center=<0 -2.2399 6.154>, rotation=<rot -0.0004 0.0002 64.9999>}]
 dyn_jacket_01_back_r [{shapeType=Capsule, length=15, radius=10.5, center=<0 -4.8495 6.135>, rotation=<rot -3.7867 4.085 -84.9859>}]
 dyn_jacket_02_back_l [{shapeType=Capsule, length=15, radius=10.5, center=<-0.7177 -3.6164 7.7554>, rotation=<rot -0.0003 0.0001 65>}]
 dyn_jacket_02_back_r [{shapeType=Capsule, length=15, radius=10.5, center=<1.6744 -4.7399 7.4684>, rotation=<rot -3.7867 4.085 -84.9859>}]
 dyn_jacket_03_back_l [{shapeType=Capsule, length=14.1686, radius=15, center=<6.6104 -2.4513 6.7349>, rotation=<rot -0.0002 0.0001 70>}]
 dyn_jacket_03_back_r [{shapeType=Capsule, length=14.1686, radius=15, center=<7.3617 -4.2635 6.6392>, rotation=<rot -3.7867 4.085 85>}]
 dyn_jacket_base_back_l [{shapeType=Capsule, length=1, radius=1}]
 dyn_jacket_base_back_r [{shapeType=Capsule, length=1, radius=1}]
 foot_l [{shapeType=Capsule, length=12.8563, radius=13.5854, center=<-2.5197 5.2749 -0.0889>, rotation=<rot 0 3.5581 -87.9004>}]
 foot_r [{shapeType=Capsule, length=12.8563, radius=13.5854, center=<2.5197 -5.2749 0.0889>, rotation=<rot 0 3.5581 92.0996>}]
 pelvis [{shapeType=Capsule, length=11.1629, radius=12.3569, shapeName=pelvis_capsule, center=<0 2.6649 0>}]
 root [{shapeType=Capsule, length=3.5, radius=195.5, shapeName=root_capsule, center=<0 0 -180>, rotation=<rot 0 -90 0>}]
 thigh_l [{shapeType=Capsule, length=20.4237, radius=10.7592, center=<-19.8798 -0.6163 0.6151>, rotation=<rot 0 -88.4152 -84.6098>}]
 thigh_r [{shapeType=Capsule, length=20.4237, radius=10.7592, center=<19.8798 0.6163 -0.6151>, rotation=<rot 0 -88.4152 95.3902>}]
}
constraints [
 {bone1Name=calf_l, bone2Name=thigh_l, swing1LimitDegrees=45, swing2LimitDegrees=45, twistLimitDegrees=45}
 {bone1Name=foot_l, bone2Name=calf_l, swing1LimitDegrees=45, swing2LimitDegrees=45, twistLimitDegrees=45}
 {bone1Name=calf_r, bone2Name=thigh_r, swing1LimitDegrees=45, swing2LimitDegrees=45, twistLimitDegrees=45}
 {bone1Name=foot_r, bone2Name=calf_r, swing1LimitDegrees=45, swing2LimitDegrees=45, twistLimitDegrees=45}
 {
  bone1Name=dyn_jacket_01_back_l
  bone2Name=dyn_jacket_base_back_l
  swing1LimitDegrees=5
  swing2LimitDegrees=20
  twistLimitDegrees=45
  twistMotion=Locked
 }
 {
  bone1Name=dyn_jacket_02_back_l
  bone2Name=dyn_jacket_01_back_l
  swing1LimitDegrees=20
  swing2LimitDegrees=35
  twistLimitDegrees=10
  twistMotion=Locked
 }
 {bone1Name=dyn_jacket_03_back_l, bone2Name=dyn_jacket_02_back_l, swing1LimitDegrees=5, swing2LimitDegrees=30, twistLimitDegrees=5}
 {
  bone1Name=thigh_l
  bone2Name=pelvis
  swing1LimitDegrees=45
  swing1Motion=Free
  swing2LimitDegrees=45
  swing2Motion=Free
  twistLimitDegrees=45
  twistMotion=Free
 }
 {
  bone1Name=thigh_r
  bone2Name=pelvis
  swing1LimitDegrees=45
  swing1Motion=Free
  swing2LimitDegrees=45
  swing2Motion=Free
  twistLimitDegrees=45
  twistMotion=Free
 }
 {
  bone1Name=dyn_jacket_base_back_l
  bone2Name=pelvis
  swing1LimitDegrees=45
  swing1Motion=Locked
  swing2LimitDegrees=45
  swing2Motion=Locked
  twistLimitDegrees=45
  twistMotion=Locked
 }
 {
  bone1Name=dyn_jacket_01_back_r
  bone2Name=dyn_jacket_base_back_r
  swing1LimitDegrees=5
  swing2LimitDegrees=20
  twistLimitDegrees=45
  twistMotion=Locked
 }
 {
  bone1Name=dyn_jacket_02_back_r
  bone2Name=dyn_jacket_01_back_r
  swing1LimitDegrees=20
  swing2LimitDegrees=35
  twistLimitDegrees=10
  twistMotion=Locked
 }
 {bone1Name=dyn_jacket_03_back_r, bone2Name=dyn_jacket_02_back_r, swing1LimitDegrees=5, swing2LimitDegrees=30, twistLimitDegrees=5}
 {
  bone1Name=dyn_jacket_base_back_r
  bone2Name=pelvis
  swing1LimitDegrees=45
  swing1Motion=Locked
  swing2LimitDegrees=45
  swing2Motion=Locked
  twistLimitDegrees=45
  twistMotion=Locked
 }
 {
  bone1Name=dyn_jacket_03_back_r
  bone2Name=dyn_jacket_03_back_l
  swing1LimitDegrees=45
  swing1Motion=Free
  swing2LimitDegrees=45
  swing2Motion=Free
  twistLimitDegrees=45
  twistMotion=Free
 }
 {bone1Name=dyn_jacket_03_back_r, bone2Name=dyn_jacket_03_back_l, swing1LimitDegrees=45, swing2LimitDegrees=45, twistLimitDegrees=45}
 {
  bone1Name=pelvis
  bone2Name=root
  swing1LimitDegrees=45
  swing1Motion=Free
  swing2LimitDegrees=45
  swing2Motion=Free
  twistLimitDegrees=45
  twistMotion=Free
 }
]
