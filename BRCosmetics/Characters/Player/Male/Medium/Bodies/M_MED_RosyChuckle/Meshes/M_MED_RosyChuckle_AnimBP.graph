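# Character-part AnimBP for the M_MED_RosyChuckle body cosmetic, adding secondary motion, mainly to the jacket.
# The parent mesh pose feeds a Control Rig, two chained pelvis-space RigidBody sims, jacket bone constraints and rotation-limited Trail controllers.
# The EventGraph derives rig inputs for sliding, gauntlets and high vertical velocity.
# It also computes a state-based Gravity that the sims, fixed at -980 Z, never read; several nodes sit disconnected.
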AnimBlueprint /BRCosmetics/Characters/Player/Male/Medium/Bodies/M_MED_RosyChuckle/Meshes/M_MED_RosyChuckle_AnimBP
target_skeleton=&/M_MED_RosyChuckle_Skeleton
graphs [
 AnimGraph {nodes [
   n1 = "Transform (Modify) Bone - Bone: dyn_jacket_base_fr_l"(Alpha=1, Translation="0,0,0", Rotation="20,0,0", Scale="1,1,1") {rotationMode=BMM_Additive, rotationSpace=BCS_ParentBoneSpace, boneToModify {boneName=dyn_jacket_base_fr_l}, rotation {pitch=20}}
   "Transform (Modify) Bone - Bone: dyn_jacket_base_side_l"(ComponentPose=@n1, Alpha=1, Translation="0,0,0", Rotation="20,0,0", Scale="1,1,1") {rotationMode=BMM_Additive, rotationSpace=BCS_ParentBoneSpace, boneToModify {boneName=dyn_jacket_base_side_l}, rotation {pitch=20}}
   RigidBody!(ComponentPose, bAlphaBoolEnabled:bool, OverrideWorldGravity:Vector="0,0,-980") -> (Pose) {
    alphaInputType=Bool
    bClampLinearTranslationLimitToRefPose=true
    bOverrideWorldGravity=true
    overridePhysicsAsset=&/M_MED_RosyChuckle_Physics_FrontEnd
    simulationSpace=BaseBoneSpace
    alphaBoolBlend {blendOption=Linear}
    alphaScaleBiasClamp {bias=1, interpSpeedDecreasing=7, interpSpeedIncreasing=15}
    baseBoneRef {boneName=pelvis}
    simSpaceSettings {dampingAlpha=0, maxAngularAcceleration=200, maxAngularVelocity=75, maxLinearAcceleration=600, maxLinearVelocity=75, worldAlpha=1}
   }
   n2 = Control Rig!(
    Source=Copy Pose From Mesh() {bUseAttachedParent=true}
    Alpha
    ISBBSliding:double=Get BBallsliding() -> (BBallsliding:double)
    IsplayingEmote:bool=Get bIsPlayingEmote()
    isGauntletsEquip:bool=Get GauntlesEquip() -> (GauntlesEquip:bool)
    DBNO:bool=Get bIsDBNO()
    isRBANon:bool=Get bRigidBodyEnabled()
    VelocityZ:double=Get TrampolineAlpha() -> (TrampolineAlpha:double)
    IsMelee:bool=Get bIsPlayingMeleeAnim()
   ) -> (Pose) {controlRigAssetReference {blueprintRigClass=&/M_MED_RosyChuckle_CtrlRig.M_MED_RosyChuckle_CtrlRig_C}}
   n3 = RigidBody!(ComponentPose=Local To Component(LocalPose=@n2), Alpha:float, OverrideWorldGravity="0,0,-980") -> (Pose) {
    bClampLinearTranslationLimitToRefPose=true
    bOverrideWorldGravity=true
    overridePhysicsAsset=&/M_MED_RosyChuckl_Pysics
    simulationSpace=BaseBoneSpace
    alphaBoolBlend {blendOption=Linear}
    alphaScaleBiasClamp {interpSpeedDecreasing=7, interpSpeedIncreasing=15}
    baseBoneRef {boneName=pelvis}
    simSpaceSettings {dampingAlpha=0, maxAngularAcceleration=75, maxAngularVelocity=35, maxLinearAcceleration=350, maxLinearVelocity=75, worldAlpha=1}
   }
   n4 = RigidBody(ComponentPose=@n3, OverrideWorldGravity="0,0,-980") {
    bClampLinearTranslationLimitToRefPose=true
    bOverrideWorldGravity=true
    overridePhysicsAsset=&/M_MED_RosyChuckle_Physics_test
    simulationSpace=BaseBoneSpace
    alphaBoolBlend {blendOption=Linear}
    alphaScaleBiasClamp {bias=1, interpSpeedDecreasing=7, interpSpeedIncreasing=15}
    baseBoneRef {boneName=pelvis}
    simSpaceSettings {dampingAlpha=0, maxAngularAcceleration=200, maxAngularVelocity=75, maxLinearAcceleration=600, maxLinearVelocity=75, worldAlpha=1}
   }
   n5 = "Constraint - dyn_jacket_04_side_l (P)"(ComponentPose=@n4, Alpha=1, ConstraintWeights_0=1) {boneToModify {boneName=dyn_jacket_04_side_l}, constraintSetup [{targetBone {boneName=dyn_jacket_03_back_l}}], constraintWeights [1]}
   n6 = "Constraint - dyn_jacket_04_side_r (P)"(ComponentPose=@n5, Alpha=1, ConstraintWeights_0=1) {boneToModify {boneName=dyn_jacket_04_side_r}, constraintSetup [{targetBone {boneName=dyn_jacket_03_back_r}}], constraintWeights [1]}
   n7 = "Constraint - dyn_jacket_base_side_r (P)"(ComponentPose=@n6, Alpha=1, ConstraintWeights_0=1) {boneToModify {boneName=dyn_jacket_base_side_r}, constraintSetup [{targetBone {boneName=dyn_jacket_base_back_r}}], constraintWeights [1]}
   n8 = Trail controller - Bone: dyn_jacket_03_back_r!(ComponentPose=@n7, bAlphaBoolEnabled:bool=True) -> (Pose) {
    alphaInputType=Bool
    bLimitRotation=true
    bShowTrailLocation=true
    maxDeltaTime=0.33
    relaxationSpeedScale=2
    fakeVelocity=<1 1 1>
    rotationLimits [{limitMax=<5 25 8>, limitMin=<-5 -15 -8>}, {limitMax=<5 25 15>, limitMin=<-5 -15 -15>}]
    rotationOffsets [<0 0 0>, <0 0 0>]
    trailBone {boneName=dyn_jacket_03_back_r}
    trailRelaxationSpeed {editorCurveData {keys [<at -0.0171 5.7275 RCIM_Linear>,<at 1.0232 9.5088 RCIM_Linear>]}}
   }
   n9 = Trail controller - Bone: dyn_jacket_03_back_l!(ComponentPose=@n8, bAlphaBoolEnabled:bool=True) -> (Pose) {
    alphaInputType=Bool
    bLimitRotation=true
    bShowTrailLocation=true
    maxDeltaTime=0.33
    relaxationSpeedScale=2
    fakeVelocity=<1 1 1>
    rotationLimits [{limitMax=<5 25 8>, limitMin=<-5 -15 -8>}, {limitMax=<5 25 15>, limitMin=<-5 -15 -15>}]
    rotationOffsets [<0 0 0>, <0 0 0>]
    trailBone {boneName=dyn_jacket_03_back_l}
    trailRelaxationSpeed {editorCurveData {keys [<at -0.0171 5.7275 RCIM_Linear>,<at 1.0232 9.5088 RCIM_Linear>]}}
   }
   n10 = Trail controller - Bone: dyn_jacket_03_back_m!(ComponentPose=@n9, bAlphaBoolEnabled:bool=True) -> (Pose) {
    alphaInputType=Bool
    bLimitRotation=true
    bShowTrailLocation=true
    maxDeltaTime=0.33
    relaxationSpeedScale=2
    fakeVelocity=<1 1 1>
    rotationLimits [{limitMax=<5 25 8>, limitMin=<-5 -15 -8>}, {limitMax=<5 25 15>, limitMin=<-5 -15 -15>}]
    rotationOffsets [<0 0 0>, <0 0 0>]
    trailBone {boneName=dyn_jacket_03_back_m}
    trailRelaxationSpeed {editorCurveData {keys [<at -0.0171 5.7275 RCIM_Linear>,<at 1.0232 9.5088 RCIM_Linear>]}}
   }
   Output Pose(Result=Component To Local(ComponentPose=@n10)) {layerGroup=None}
  ]}
 EventGraph {nodes [
   Event BlueprintUpdateAnimation()
   > n1 = Sequence!(execute) -> (then_0, then_1, then_2:Exec, then_3:Exec)
   n2 = Get As Fort Player Pawn Athena() -> (As Fort Player Pawn Athena:Fort Player Pawn Athena Object Reference)
   > Is Valid(InputObject=@n2)
   n3 = GetDisplayName(Object=GetSkeletalMeshForPartType(self=@n2, PartType=Gameplay))
   > Set GauntlesEquip(GauntlesEquip:bool=OR Boolean!(
     A=Contains(SearchIn=@n3, Substring=GauntletsTemplate_Gadget)
     B=Contains(SearchIn=@n3, Substring=SprocketPoppy_Gauntlet)
     C:bool=Contains(SearchIn=@n3, Substring=CoffeePuff_Gauntlet)
    ) -> (ReturnValue)) -> (Output_Get:bool)
   n4 = Branch(execute=@n1.then_3, Condition="float > float"(A=Get PawnVelocityZ(), B=700))
   > Set TrampolineAlpha(TrampolineAlpha:double=1) -> (Output_Get:double)
   Set TrampolineAlpha(execute=@n4.else, TrampolineAlpha=0)
   n5 = Gravity Override State Function DEPRECATED(
    execute=@n1.then_1
    CharacterPartAnimInstance=Self-Reference()
    GameTimeInSeconds=GetGameTimeInSeconds()
    GravityOverrideParams=Get Gravity Override Params()
    RBAN Params=Get RBAN Params()
   )
   > Set Gravity(Gravity:Vector=@n5) -> (Output_Get:Vector)
   > Set bISRbanOFF(bISRbanOFF:bool=NOT Boolean(A=Get bRigidBodyEnabled())) -> (Output_Get:bool)
   n6 = Event OnBeginBaseballSliding() {is=title K2Node_Event}
   Set BBallsliding(execute=[@n6.then, @n9.then], BBallsliding:double=1) -> (Output_Get:double)
   Event OnCharacterCustomizationLoaded()
   n7 = TryGetPawnOwner()
   > Is Valid(InputObject=@n7)
   > n8 = Cast To FortPlayerPawnAthena(Object=@n7)
   > Set As Fort Player Pawn Athena(As Fort Player Pawn Athena:Fort Player Pawn Athena Object Reference=@n8) -> (Output_Get:Fort Player Pawn Athena Object Reference)
   n9 = Event OnBeginKneeSliding() {is=title K2Node_Event}
   n10 = Event OnEndBaseballSliding() {is=title K2Node_Event}
   Set BBallsliding(execute=[@n10.then, @n11.then], BBallsliding=0)
   n11 = Event OnEndKneeSliding() {is=title K2Node_Event}
   n12 = Branch(Condition=Get bIsPlayingMeleeAnim())
   n13 = Get As Fort Player Pawn Athena()
   n14 = BreakRotator(InRot=GetDeltaTransformFromRefPose!(self=Get Mesh(self=@n13), BoneName=spine_05, BaseName=spine_01) -> (ReturnValue_Location:Vector, ReturnValue_Rotation:Rotator, ReturnValue_Scale:Vector).ReturnValue_Rotation)
   > Set Swing(Swing:double=MapRangeClamped(Value=@n14, InRangeA=0, InRangeB=20)) -> (Output_Get:double)
   Set Swing(execute=@n12.else, Swing=FInterpTo(Current=Get Swing() -> (Swing:double), Target=0, DeltaTime=GetGameTimeInSeconds(), InterpSpeed=20))
   n15 = Get bIsPlayingEmote()
   Set bIFEBKnoEMOTE(bIFEBKnoEMOTE:bool=AND Boolean(A=Get bIsInFrontEnd(), B=NOT Boolean(A=@n15))) -> (Output_Get:bool)
   > n16 = Branch(Condition=Get bIFEBKnoEMOTE() -> (bIFEBKnoEMOTE:bool))
   > Set BKAlpha(BKAlpha:double=0) -> (Output_Get:double)
   Set BKAlpha(execute=@n16.else, BKAlpha=1)
   GetOwnerSkeletalMeshComponent()
   Get bIsPlayingEmote()
   Get bIsSprinting()
   Get Mesh(self=Get As Fort Player Pawn Athena())
   Get Gravity Override Params()
   Set Swing(Swing=MapRangeClamped(Value=@n14, InRangeA=0, InRangeB=-20))
  ]}
]
variables [
 As Fort Player Pawn Athena {type=FortPlayerPawnAthena}
 BBallsliding {type=number}
 bIFEBKnoEMOTE {type=boolean}
 biFEIKemote {type=boolean}
 bISRbanOFF {type=boolean}
 BKAlpha {type=number}
 GauntlesEquip {type=boolean}
 Gravity {type=Vector}
 Gravity Override Params {
  type=GravityOverrideParamsStruct
  default {
   dBNO {gravityOverride=<-700 400 0>}
   onGround_Moving {gravityOverride=<0 -100 -980>}
   skydive_DivingDown {gravityOverride=<-1000 0 -500>}
   skydive_FloatingDown {gravityOverride=<0 -700 -300>}
   skydive_GoingUp {gravityOverride=<0 -300 -400>}
   skydive_Parachuting {gravityOverride=<0 0 -980>}
   tactical_Sprint {gravityOverride=<0 -330 -500>}
  }
 }
 KneeSliding {type=number}
 RBAN Params {
  type=RBANParamsStruct
  default {
   dBNO {simSpaceSettings {
     maxAngularAcceleration=1000
     maxAngularVelocity=3
     maxLinearAcceleration=300
     maxLinearVelocity=700
     velocityScaleZ=0
     externalLinearDragV=<0 0.5 0>
    }}
   dBNO_Carried {simSpaceSettings {
     maxAngularAcceleration=1000
     maxAngularVelocity=3
     maxLinearAcceleration=300
     maxLinearVelocity=700
     velocityScaleZ=0
     externalLinearDragV=<0 0.5 0>
    }}
   emoteOrMelee {simSpaceSettings {
     maxAngularAcceleration=1000
     maxAngularVelocity=3
     maxLinearAcceleration=300
     maxLinearVelocity=700
     velocityScaleZ=0
     externalLinearDragV=<0 0.5 0>
    }}
   front_End {simSpaceSettings {
     maxAngularAcceleration=1000
     maxAngularVelocity=3
     maxLinearAcceleration=300
     maxLinearVelocity=700
     velocityScaleZ=0
     externalLinearDragV=<0 0.5 0>
    }}
   onGround_Crouch_Moving {simSpaceSettings {
     maxAngularAcceleration=1000
     maxAngularVelocity=3
     maxLinearAcceleration=300
     maxLinearVelocity=700
     velocityScaleZ=0
     externalLinearDragV=<0 0.5 0>
    }}
   onGround_Falling {simSpaceSettings {
     maxAngularAcceleration=1000
     maxAngularVelocity=3
     maxLinearAcceleration=300
     maxLinearVelocity=700
     velocityScaleZ=0
     externalLinearDragV=<0 0.5 0>
    }}
   onGround_Moving {simSpaceSettings {
     maxAngularAcceleration=1000
     maxAngularVelocity=3
     maxLinearAcceleration=300
     maxLinearVelocity=700
     velocityScaleZ=0
     externalLinearDragV=<0 0.5 0>
    }}
   onGround_Standing {simSpaceSettings {
     maxAngularAcceleration=1000
     maxAngularVelocity=3
     maxLinearAcceleration=300
     maxLinearVelocity=700
     velocityScaleZ=0
     externalLinearDragV=<0 0.5 0>
    }}
   rail_Grinding {simSpaceSettings {
     maxAngularAcceleration=1000
     maxAngularVelocity=3
     maxLinearAcceleration=300
     maxLinearVelocity=700
     velocityScaleZ=0
     worldAlpha=0.1
     externalLinearDragV=<0 0.5 0>
    }}
   riding_Moving {simSpaceSettings {
     maxAngularAcceleration=1000
     maxAngularVelocity=3
     maxLinearAcceleration=300
     maxLinearVelocity=700
     velocityScaleZ=0
     externalLinearDragV=<0 0.5 0>
    }}
   riding_Moving_Targeting {simSpaceSettings {
     maxAngularAcceleration=1000
     maxAngularVelocity=3
     maxLinearAcceleration=300
     maxLinearVelocity=700
     velocityScaleZ=0
     externalLinearDragV=<0 0.5 0>
    }}
   riding_Standing {simSpaceSettings {
     maxAngularAcceleration=1000
     maxAngularVelocity=3
     maxLinearAcceleration=300
     maxLinearVelocity=700
     velocityScaleZ=0
     externalLinearDragV=<0 0.5 0>
    }}
   skydive_DivingDown {simSpaceSettings {
     maxAngularAcceleration=1000
     maxAngularVelocity=3
     maxLinearAcceleration=300
     maxLinearVelocity=700
     velocityScaleZ=0
     externalLinearDragV=<0 0.5 0>
    }}
   skydive_FloatingDown {
    componentLinearAccScale=<0 0.5 0>
    componentLinearVelScale=<0 0.5 0>
    simSpaceSettings {
     maxAngularAcceleration=1000
     maxAngularVelocity=3
     maxLinearAcceleration=300
     maxLinearVelocity=700
     velocityScaleZ=0
     externalLinearDragV=<0 0.5 0>
    }
   }
   skydive_GoingUp {simSpaceSettings {
     maxAngularAcceleration=1000
     maxAngularVelocity=3
     maxLinearAcceleration=300
     maxLinearVelocity=700
     velocityScaleZ=0
     externalLinearDragV=<0 0.5 0>
    }}
   skydive_Parachuting {
    componentLinearAccScale=<0 0.1 0.3>
    componentLinearVelScale=<0 0.1 0.3>
    simSpaceSettings {
     maxAngularAcceleration=1000
     maxAngularVelocity=3
     maxLinearAcceleration=300
     maxLinearVelocity=700
     velocityScaleZ=0
     externalLinearDragV=<0 0.5 0>
    }
   }
   skydive_Surfing {simSpaceSettings {
     maxAngularAcceleration=1000
     maxAngularVelocity=3
     maxLinearAcceleration=300
     maxLinearVelocity=700
     velocityScaleZ=0
     externalLinearDragV=<0 0.5 0>
    }}
   swimming {simSpaceSettings {
     maxAngularAcceleration=1000
     maxAngularVelocity=3
     maxLinearAcceleration=300
     maxLinearVelocity=700
     velocityScaleZ=0
     externalLinearDragV=<0 0.5 0>
    }}
   tactical_Sprint {simSpaceSettings {
     maxAngularAcceleration=1000
     maxAngularVelocity=3
     maxLinearAcceleration=300
     maxLinearVelocity=700
     velocityScaleZ=0
     externalLinearDragV=<0 0.5 0>
    }}
   ziplining {simSpaceSettings {
     maxAngularAcceleration=1000
     maxAngularVelocity=3
     maxLinearAcceleration=300
     maxLinearVelocity=700
     velocityScaleZ=0
     externalLinearDragV=<0 0.5 0>
    }}
  }
 }
 Swing {type=number}
 TrampolineAlpha {type=number}
]
