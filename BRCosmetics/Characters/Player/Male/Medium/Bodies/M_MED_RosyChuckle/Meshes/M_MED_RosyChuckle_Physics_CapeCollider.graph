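# Physics asset for the RosyChuckle male medium body, simulating its back jacket or cape bones against body colliders.
# Left and right three-segment capsule chains hang from small base bodies rotationally locked to root, with limited swing and mostly locked twist, and cross-linked at their tips.
# Capsules on upper spine, arms, hands, legs and feet form the body colliders.
# A huge root capsule offset downward likely acts as a floor.
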PhysicsAsset /BRCosmetics/Characters/Player/Male/Medium/Bodies/M_MED_RosyChuckle/Meshes/M_MED_RosyChuckle_Physics_CapeCollider
bodies {
 calf_l [{shapeType=Capsule, length=20.6458, radius=10.3826, center=<-19.7227 -1.2259 -0.5968>, rotation=<rot 0 -91.8539 -89.8577>}]
 calf_r [{shapeType=Capsule, length=20.1555, radius=10.3242, center=<20.2878 1.3914 0.5245>, rotation=<rot 0 -90.4056 -88.8823>}]
 dyn_jacket_01_back_l [{shapeType=Capsule, length=15, radius=10.5, center=<2.9821 6.1671 -2.5872>, rotation=<rot -0.0003 0.0001 70>}]
 dyn_jacket_01_back_r [{shapeType=Capsule, length=15, radius=10.5, center=<3.6021 -6.5218 -3.5972>, rotation=<rot -3.7867 4.085 -84.9859>}]
 dyn_jacket_02_back_l [{shapeType=Capsule, length=15, radius=10.5, center=<2.2643 6.9772 -3.344>, rotation=<rot -0.0003 0.0001 65>}]
 dyn_jacket_02_back_r [{shapeType=Capsule, length=15, radius=10.5, center=<3.3479 -6.6567 -4.9096>, rotation=<rot -3.7867 4.085 -84.9859>}]
 dyn_jacket_03_back_l [{shapeType=Capsule, length=15, radius=10, center=<9.5925 6.225 -2.7464>, rotation=<rot -0.0002 0.0001 70>}]
 dyn_jacket_03_back_r [{shapeType=Capsule, length=15, radius=10, center=<9.1489 -6.1257 -3.9653>, rotation=<rot -3.7867 4.085 -84.9859>}]
 dyn_jacket_base_back_l [{shapeType=Capsule, length=1, radius=1}]
 dyn_jacket_base_back_r [{shapeType=Capsule, length=1, radius=1}]
 foot_l [{shapeType=Capsule, length=12.8563, radius=13.5854, center=<-2.5197 5.2749 -0.0889>, rotation=<rot 0 3.5581 -87.9004>}]
 foot_r [{shapeType=Capsule, length=12.8563, radius=13.5854, center=<2.5197 -5.2749 0.0889>, rotation=<rot 0 3.5581 92.0996>}]
 hand_l [{shapeType=Capsule, length=9.5015, radius=6.7933, center=<6.2703 0 0>, rotation=<rot 90 0 0>}]
 hand_r [{shapeType=Capsule, length=9.5015, radius=6.7933, center=<-6.2703 0 0>, rotation=<rot 90 0 180>}]
 lowerarm_l [{shapeType=Capsule, length=12.2721, radius=9.624, center=<8.9852 0 1.9278>, rotation=<rot 90 0 0>}]
 lowerarm_r [{shapeType=Capsule, length=12.2721, radius=9.624, center=<-8.9852 0 -1.9278>, rotation=<rot 90 0 180>}]
 root [{shapeType=Capsule, length=3.5, radius=195.5, shapeName=root_capsule, center=<0 0 -180>, rotation=<rot 0 -90 0>}]
 spine_04 [{shapeType=Capsule, length=21.5471, radius=20.3326, center=<0 3.3127 0>}]
 spine_05 [{shapeType=Capsule, length=18.2569, radius=19.2748, center=<0 3.0766 0>}]
 thigh_l [{shapeType=Capsule, length=20.4237, radius=10.7592, center=<-19.8798 -0.6163 0.6151>, rotation=<rot 0 -88.4152 -84.6098>}]
 thigh_r [{shapeType=Capsule, length=20.4237, radius=10.7592, center=<19.8798 0.6163 -0.6151>, rotation=<rot 0 -88.4152 95.3902>}]
 upperarm_l [{shapeType=Capsule, length=11.1688, radius=13.7525, center=<10.0455 0 0>, rotation=<rot 90 0 0>}]
 upperarm_r [{shapeType=Capsule, length=11.1688, radius=13.7525, center=<-10.0455 0 0>, rotation=<rot 90 0 180>}]
}
constraints [
 {bone1Name=calf_l, bone2Name=thigh_l, swing1LimitDegrees=45, swing2LimitDegrees=45, twistLimitDegrees=45}
 {bone1Name=foot_l, bone2Name=calf_l, swing1LimitDegrees=45, swing2LimitDegrees=45, twistLimitDegrees=45}
 {bone1Name=calf_r, bone2Name=thigh_r, swing1LimitDegrees=45, swing2LimitDegrees=45, twistLimitDegrees=45}
 {bone1Name=foot_r, bone2Name=calf_r, swing1LimitDegrees=45, swing2LimitDegrees=45, twistLimitDegrees=45}
 {
  bone1Name=dyn_jacket_01_back_l
  bone2Name=dyn_jacket_base_back_l
  swing1LimitDegrees=5
  swing2LimitDegrees=20
  twistLimitDegrees=45
  twistMotion=Locked
 }
 {
  bone1Name=dyn_jacket_02_back_l
  bone2Name=dyn_jacket_01_back_l
  swing1LimitDegrees=20
  swing2LimitDegrees=35
  twistLimitDegrees=10
  twistMotion=Locked
 }
 {bone1Name=dyn_jacket_03_back_l, bone2Name=dyn_jacket_02_back_l, swing1LimitDegrees=5, swing2LimitDegrees=30, twistLimitDegrees=5}
 {
  bone1Name=thigh_l
  bone2Name=root
  swing1LimitDegrees=45
  swing1Motion=Free
  swing2LimitDegrees=45
  swing2Motion=Free
  twistLimitDegrees=45
  twistMotion=Free
 }
 {
  bone1Name=thigh_r
  bone2Name=root
  swing1LimitDegrees=45
  swing1Motion=Free
  swing2LimitDegrees=45
  swing2Motion=Free
  twistLimitDegrees=45
  twistMotion=Free
 }
 {
  bone1Name=dyn_jacket_base_back_l
  bone2Name=root
  swing1LimitDegrees=45
  swing1Motion=Locked
  swing2LimitDegrees=45
  swing2Motion=Locked
  twistLimitDegrees=45
  twistMotion=Locked
 }
 {
  bone1Name=dyn_jacket_01_back_r
  bone2Name=dyn_jacket_base_back_r
  swing1LimitDegrees=5
  swing2LimitDegrees=20
  twistLimitDegrees=45
  twistMotion=Locked
 }
 {
  bone1Name=dyn_jacket_02_back_r
  bone2Name=dyn_jacket_01_back_r
  swing1LimitDegrees=20
  swing2LimitDegrees=35
  twistLimitDegrees=10
  twistMotion=Locked
 }
 {bone1Name=dyn_jacket_03_back_r, bone2Name=dyn_jacket_02_back_r, swing1LimitDegrees=5, swing2LimitDegrees=30, twistLimitDegrees=5}
 {
  bone1Name=dyn_jacket_base_back_r
  bone2Name=root
  swing1LimitDegrees=45
  swing1Motion=Locked
  swing2LimitDegrees=45
  swing2Motion=Locked
  twistLimitDegrees=45
  twistMotion=Locked
 }
 {
  bone1Name=dyn_jacket_03_back_r
  bone2Name=dyn_jacket_03_back_l
  swing1LimitDegrees=45
  swing1Motion=Free
  swing2LimitDegrees=45
  swing2Motion=Free
  twistLimitDegrees=45
  twistMotion=Free
 }
 {bone1Name=dyn_jacket_03_back_r, bone2Name=dyn_jacket_03_back_l, swing1LimitDegrees=45, swing2LimitDegrees=45, twistLimitDegrees=45}
 {
  bone1Name=spine_04
  bone2Name=root
  swing1LimitDegrees=45
  swing1Motion=Free
  swing2LimitDegrees=45
  swing2Motion=Free
  twistLimitDegrees=45
  twistMotion=Free
 }
 {
  bone1Name=spine_05
  bone2Name=spine_04
  swing1LimitDegrees=45
  swing1Motion=Free
  swing2LimitDegrees=45
  swing2Motion=Free
  twistLimitDegrees=45
  twistMotion=Free
 }
 {
  bone1Name=upperarm_l
  bone2Name=spine_05
  swing1LimitDegrees=45
  swing1Motion=Free
  swing2LimitDegrees=45
  swing2Motion=Free
  twistLimitDegrees=45
  twistMotion=Free
 }
 {
  bone1Name=lowerarm_l
  bone2Name=upperarm_l
  swing1LimitDegrees=45
  swing1Motion=Free
  swing2LimitDegrees=45
  swing2Motion=Free
  twistLimitDegrees=45
  twistMotion=Free
 }
 {
  bone1Name=hand_l
  bone2Name=lowerarm_l
  swing1LimitDegrees=45
  swing1Motion=Free
  swing2LimitDegrees=45
  swing2Motion=Free
  twistLimitDegrees=45
  twistMotion=Free
 }
 {
  bone1Name=upperarm_r
  bone2Name=spine_05
  swing1LimitDegrees=45
  swing1Motion=Free
  swing2LimitDegrees=45
  swing2Motion=Free
  twistLimitDegrees=45
  twistMotion=Free
 }
 {
  bone1Name=lowerarm_r
  bone2Name=upperarm_r
  swing1LimitDegrees=45
  swing1Motion=Free
  swing2LimitDegrees=45
  swing2Motion=Free
  twistLimitDegrees=45
  twistMotion=Free
 }
 {
  bone1Name=hand_r
  bone2Name=lowerarm_r
  swing1LimitDegrees=45
  swing1Motion=Free
  swing2LimitDegrees=45
  swing2Motion=Free
  twistLimitDegrees=45
  twistMotion=Free
 }
]
