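# Face accessory AnimBP for the RosyChuckle player body, layering procedural motion on the pose copied from the attached parent mesh.
# The copied pose feeds a Control Rig, then spring-driven AnimDynamics for beard and hair bones.
# A weighted constraint makes Base_candy follow head, lip and jaw bones, mostly the lower lip.
# In the front end, the Frontend_Idle_Bypass curve weights a DefaultSlot blend from the Glasses bone down.
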AnimBlueprint /BRCosmetics/Characters/Player/Male/Medium/Bodies/M_MED_RosyChuckle/Meshes/Parts/M_MED_RosyChuckle_FaceAcc_AnimBP
target_skeleton=&/M_MED_RosyChuckle_FaceAcc_Skeleton
graphs [
 AnimGraph {nodes [
   n1 = Control Rig(Source=Copy Pose From Mesh() {bUseAttachedParent=true}) {controlRigAssetReference {blueprintRigClass=&/M_MED_RosyChuckle_FaceAcc_CtrlRig.M_MED_RosyChuckle_FaceAcc_CtrlRig_C}}
   n2 = Anim Dynamics - Bone: Dyn_beard_01(ComponentPose=Local To Component(LocalPose=@n1), Alpha=1, ExternalForce="0,0,0") {
    angularDampingOverride=0.9999
    angularSpringConstant=50
    bAngularSpring=true
    bOverrideAngularDamping=true
    bUseGravityOverride=true
    boundBone {boneName=Dyn_beard_01}
    gravityOverride {z=-980}
    physicsBodyDefinitions [{
      boundBone {boneName=Dyn_beard_01}
      boxExtents=<10 10 10>
      constraintSetup {angularLimitsMax=<0 25 10>, angularLimitsMin=<0 2 -10>, angularTarget=<0 0 0>}
      localJointOffset=<0.5 0 0>
     }]
   }
   n3 = Anim Dynamics - Bone: Dyn_beard_02(ComponentPose=@n2, Alpha=1, ExternalForce="0,0,0") {
    angularDampingOverride=0.9999
    angularSpringConstant=50
    bAngularSpring=true
    bOverrideAngularDamping=true
    bUseGravityOverride=true
    boundBone {boneName=Dyn_beard_02}
    gravityOverride {z=-980}
    physicsBodyDefinitions [{
      boundBone {boneName=Dyn_beard_02}
      boxExtents=<10 10 10>
      constraintSetup {angularLimitsMax=<0 25 10>, angularLimitsMin=<0 2 -10>, angularTarget=<0 0 0>}
      localJointOffset=<0.3 0 0>
     }]
   }
   n4 = Anim Dynamics - Bone: Hair_A(ComponentPose=@n3, Alpha=1, ExternalForce="0,0,0") {
    angularDampingOverride=0.999
    angularSpringConstant=50
    bAngularSpring=true
    bGravityOverrideInSimSpace=true
    bOverrideAngularDamping=true
    bUseGravityOverride=true
    simulationSpace=World
    boundBone {boneName=Hair_A}
    componentAppliedLinearAccClamp=<5000 5000 5000>
    componentLinearAccScale=<0.4 0.4 0.3>
    componentLinearVelScale=<0.4 0.4 0.3>
    gravityOverride {z=-980}
    physicsBodyDefinitions [{
      boundBone {boneName=Hair_A}
      boxExtents=<10 10 10>
      constraintSetup {angularLimitsMax=<0 -3 20>, angularLimitsMin=<0 -20 -20>, angularTarget=<0 0 0>}
      localJointOffset=<1 0 0>
     }]
   }
   n5 = Anim Dynamics - Bone: Hair_B(ComponentPose=@n4, Alpha=1, ExternalForce="0,0,0") {
    angularDampingOverride=0.999
    angularSpringConstant=50
    bAngularSpring=true
    bGravityOverrideInSimSpace=true
    bOverrideAngularDamping=true
    bUseGravityOverride=true
    simulationSpace=World
    boundBone {boneName=Hair_B}
    componentAppliedLinearAccClamp=<5000 5000 5000>
    componentLinearAccScale=<0.4 0.4 0.3>
    componentLinearVelScale=<0.4 0.4 0.3>
    gravityOverride {z=-980}
    physicsBodyDefinitions [{
      boundBone {boneName=Hair_B}
      boxExtents=<5 5 5>
      constraintSetup {angularLimitsMax=<0 -3 20>, angularLimitsMin=<0 -20 -20>, angularTarget=<0 0 0>}
      localJointOffset=<1 0 0>
     }]
   }
   n6 = Anim Dynamics - Bone: Hair_C(ComponentPose=@n5, Alpha=1, ExternalForce="0,0,0") {
    angularDampingOverride=0.9999
    angularSpringConstant=50
    bAngularSpring=true
    bOverrideAngularDamping=true
    bUseGravityOverride=true
    boundBone {boneName=Hair_C}
    componentAppliedLinearAccClamp {z=5000}
    componentLinearAccScale=<0.4 0.4 0.3>
    componentLinearVelScale=<0.4 0.4 0.3>
    gravityOverride {z=-980}
    physicsBodyDefinitions [{
      boundBone {boneName=Hair_C}
      boxExtents=<5 5 5>
      constraintSetup {angularLimitsMax=<0 10 20>, angularLimitsMin=<0 -10 -20>, angularTarget=<0 0 0>}
      localJointOffset=<-1 0 0>
     }]
   }
   n7 = "Constraint - Base_candy (PPPPPPP)"!(
    ComponentPose=@n6
    Alpha=1
    ConstraintWeights_0=0.0013
    ConstraintWeights_1:float=0.0199
    ConstraintWeights_2:float=0.5923
    ConstraintWeights_3:float=0.0088
    ConstraintWeights_4:float=0.2912
    ConstraintWeights_5:float=0.0301
    ConstraintWeights_6:float=0.0565
   ) -> (Pose) {
    boneToModify {boneName=Base_candy}
    constraintSetup [
     {targetBone {boneName=head}}
     {targetBone {boneName=FACIAL_R_LipCorner}}
     {targetBone {boneName=FACIAL_C_LipLower}}
     {targetBone {boneName=FACIAL_L_LipLower}}
     {targetBone {boneName=FACIAL_R_LipLower}}
     {targetBone {boneName=FACIAL_R_LipLowerOuter}}
     {targetBone {boneName=FACIAL_C_Jaw}}
    ]
    constraintWeights [1,1,1,1,1,1,1]
   }
   Output Pose(Result=Layered blend per bone(
     BasePose=Component To Local(ComponentPose=@n7)
     BlendPoses_0=Slot 'DefaultSlot'()
     BlendWeights_0=Get Custom Idle is Playing Float() -> (Custom Idle is Playing Float:double)
    ) {layerSetup [{branchFilters [{boneName=Glasses}]}]}) {layerGroup=None}
  ]}
 EventGraph {nodes [
   Event BlueprintUpdateAnimation()
   > Branch(Condition=Get bIsInFrontEnd())
   > Set Custom Idle is Playing Float(Custom Idle is Playing Float:double=GetCurveValue(self=GetOwnerFrontEndAnimInstance(), CurveName=Frontend_Idle_Bypass)) -> (Output_Get:double)
  ]}
]
variables [Custom Idle is Playing Float {type=number}]
